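# Test Niagara system for the Halo Flux vehicle boost cosmetic, built from several mesh, sprite and ribbon emitters.
# Boost_Rings continuously spawns particles at skeletal mesh locations plus a car offset, sends them backward and colors them from a palette array.
# DiskCenter and BeeHive burst-spawn rotating center elements with paint-variant overrides.
# Spawn-group masks select per-layer settings; the root and sine-wave black-flame ribbon emitters are disabled.
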NiagaraSystem /VehicleCosmetics/Boosts/Boost_HaloFlux/FX/NS_Boost_HaloFlux_TEST
dependencies {
 usedDataInterfaces [
  &NiagaraDataInterfaceArrayColor
  &NiagaraDataInterfaceVectorCurve
  &NiagaraDataInterfaceCurve
  &NiagaraDataInterfaceColorCurve
  &NiagaraDataInterfaceArrayFloat
  &NiagaraDataInterfaceCamera
  &NiagaraDataInterfaceSkeletalMesh
 ]
 usedDynamicInputs [&MakeFloatFromVector]
 usedModules [
  &self:SystemSpawnScript.SystemScriptSource.SystemScriptGraph.NiagaraNodeAssignment_0.SetVariables_#1
  &SystemState
 ]
}
emitters {
 Boost_Rings {
  emitterUpdateScript [
   &EmitterState(Inactive Response=::NewEnumerator2, Life Cycle Mode=::NewEnumerator1, Loop Duration=1, UseLoopDelay=false)
   &VC_FilteredSocketCount()
   &SpawnRate(SpawnRate=64, Use Spawn Probability=false)
   &SpawnRate(SpawnRate=32, Use Spawn Probability=false) {enabled=false, module=SpawnRate001}
  ]
  particleSpawnScript [
   /SetVariables(
    Particles.ColorRings:LinearColor=&SelectLinearColorFromArray(
     Array Sampling Mode=::NewEnumerator0
     Direct Array Index=?
     Color Selection Array=DataInterface() {propertyValues {ColorData [
        <rgba 0.972 3 0.45 1>
        <rgba 1.5 0 0 1>
        <rgba 0.141 0.222 3 1>
        <rgba 0 2.544 1 1>
        <rgba 3 1.293 0.132 1>
        <rgba 3 2.625 0.003 1>
        <rgba 0.24 3 0.669 1>
       ]}}
     Random Seed=0
    )
    Particles.RenderTag:NiagaraInt32=&SpawnGroup()
   ) {isSetParameters=true, module=SetVariables_#2}
   &VC_PaintedVariantsOverride() {enabled=false}
   &InitializeParticle(
    Color=$Particles.ColorRings
    Lifetime=&MaskFloatBySpawnGroup(
     Masked Float=0.75
     Passthrough Float=&MaskFloatBySpawnGroup(
      Masked Float=1
      Passthrough Float=&MaskFloatBySpawnGroup(
       Masked Float=1
       Passthrough Float=&MaskFloatBySpawnGroup(Masked Float=0.4, Passthrough Float=1, Spawn Group Mask=3)
       Spawn Group Mask=2
      )
      Spawn Group Mask=1
     )
    )
    Mesh Scale Mode=::NewEnumerator1
    Mesh Uniform Scale Max=0.6
    Mesh Uniform Scale Min=0.15
    Sprite Rotation Mode=::NewEnumerator2
    UsePositionOffset=false
    Write Material Random=true
   )
   &PartitionParticles(
    Particles Per Partition=0
    Partition Count Mode:ENiagara_InfiniteLoopDuration=::NewEnumerator1
    Partition Mode:ENiagara_PartitionMode=::NewEnumerator1
   ) {enabled=false}
   &InitialMeshOrientation(
    Mesh Axis To Orient=<0 0 0>
    Mesh Orientation Mode=::NewEnumerator2
    Orientation Vector=&MaskVectorBySpawnGroup(
     Masked Vector=<0 0 0.5>
     Passthrough Vector=&MaskVectorBySpawnGroup(
      Masked Vector=<0 0 0.5>
      Passthrough Vector=&MaskVectorBySpawnGroup(
       Masked Vector=<0 0 0.5>
       Passthrough Vector=&MaskVectorBySpawnGroup(Masked Vector=<0 0 0>, Passthrough Vector=<1 1 1>, Spawn Group Mask=3)
       Spawn Group Mask=2
      )
      Spawn Group Mask=1
     )
    )
    Rotation=&RandomRangeVector(
     Minimum=&MaskVectorBySpawnGroup(
      Masked Vector=<0 0 0>
      Passthrough Vector=&MaskVectorBySpawnGroup(
       Masked Vector=<0 0 -1>
       Passthrough Vector=&MaskVectorBySpawnGroup(
        Masked Vector=<0 0 -1>
        Passthrough Vector=&MaskVectorBySpawnGroup(Masked Vector=<-1 0 0>, Passthrough Vector=<1 1 1>, Spawn Group Mask=3)
        Spawn Group Mask=2
       )
       Spawn Group Mask=1
      )
     )
     Maximum=&MaskVectorBySpawnGroup(
      Masked Vector=<0 0 0>
      Passthrough Vector=&MaskVectorBySpawnGroup(
       Masked Vector=<0 0 1>
       Passthrough Vector=&MaskVectorBySpawnGroup(
        Masked Vector=<0 0 1>
        Passthrough Vector=&MaskVectorBySpawnGroup(Masked Vector=<1 0 0>, Passthrough Vector=<1 1 1>, Spawn Group Mask=3)
        Spawn Group Mask=2
       )
       Spawn Group Mask=1
      )
     )
    )
    Use Rotation Vector=true
   )
   &SkeletalMeshLocation(
    Bone / Socket Index=$Particles.PartitionParticles.Partition
    Bone / Socket Sampling Mode=::NewEnumerator5
    MaskSpawnGroup=false
    Mesh Sampling Type=::NewEnumerator6
    Sampled Position Offset=&AddVectorToPosition(
     Position=&MaskVectorBySpawnGroup(
      Masked Vector=<-50 0 0>
      Passthrough Vector=&MaskVectorBySpawnGroup(
       Masked Vector=<0 0 0>
       Passthrough Vector=&MaskVectorBySpawnGroup(
        Masked Vector=<0 0 0>
        Passthrough Vector=&MaskVectorBySpawnGroup(Masked Vector=<0 0 0>, Passthrough Vector=<1 1 1>, Spawn Group Mask=3)
        Spawn Group Mask=2
       )
       Spawn Group Mask=1
      )
     )
     Vector=Linked() {linkedVariable {name=User.DTCarOffset, type=NiagaraPosition}}
    )
    Skeletal Mesh=$System.SkeletalMesh
   )
   &AddVelocity(Velocity=&MaskVectorBySpawnGroup(Masked Vector=<-1250 0 0>, Passthrough Vector=<1 1 1>))
   &ShapeLocation(MaskSpawnGroup=false, Offset=<-67.5708 0 0>, Offset Mode=::NewEnumerator0, Override Seed=false, Sphere Radius=0, UseOwnerScale=true)
  ]
  particleUpdateScript [
   &ParticleState()
   &UpdateMeshOrientation(Rotation Vector=<0 0 2>) {enabled=false}
   &ScaleMeshSize(Scale Factor=&MaskVectorBySpawnGroup(
     Masked Vector=&VectorFromCurve(VectorCurve=DataInterface() {propertyValues {
        XCurve {keys [<at 0 0>,<at 0.1472 0.6951>,<at 0.37 0.8221>,<at 0.7541 0.5672>,<at 1 0>]}
        YCurve {keys [<at 0 0>,<at 0.1472 0.6951>,<at 0.3659 0.832>,<at 0.7541 0.5672>,<at 1 0>]}
        ZCurve {keys [<at 0 0>,<at 0.1472 0.6951>,<at 0.37 0.8418>,<at 0.7541 0.5672>,<at 1 0>]}
       }})
     Passthrough Vector=&MaskVectorBySpawnGroup(
      Masked Vector=&VectorFromCurve(VectorCurve=DataInterface() {propertyValues {
         XCurve {keys [<at 0.0085 0.7461>,<at 0.1197 0.9841>,<at 0.5195 0.898>,<at 0.9973 0.1466>]}
         YCurve {keys [<at 0.0017 0.7193>,<at 0.1289 0.9841>,<at 0.5226 0.9177>,<at 0.9973 0.1466>]}
         ZCurve {keys [<at 0.0022 0.7717>,<at 0.1517 1.1096>,<at 0.5124 1.0666>,<at 1.0156 0.1427>]}
        }})
      Passthrough Vector=&MaskVectorBySpawnGroup(
       Masked Vector=&VectorFromCurve(VectorCurve=DataInterface() {propertyValues {
          XCurve {keys [<at -0.0069 0.4922>,<at 0.0243 0.6397>,<at 0.495 0.6299>,<at 0.5 0.57>,<at 0.9519 0.57>,<at 1 0>]}
          YCurve {keys [<at -0.0069 0.4922>,<at 0.0377 0.6397>,<at 0.495 0.6299>,<at 0.5 0.57>,<at 0.9361 0.57>,<at 1 0>]}
          ZCurve {keys [<at 0 0.5668>,<at 0.0432 0.7017>,<at 0.495 0.7017>,<at 0.5 0.78>,<at 0.8305 0.78>,<at 1.0139 0.75>]}
         }})
       Passthrough Vector=&MaskVectorBySpawnGroup(
        Masked Vector=&VectorFromFloat(Value=&FloatFromCurve(FloatCurve=DataInterface() {propertyValues {Curve {keys [
              <at 0 0.2824>
              {arriveTangent=0.049, interpMode=RCIM_Cubic, leaveTangent=0.049, tangentMode=RCTM_User, time=0.2283, value=0.9882}
              {arriveTangent=-1.3884, interpMode=RCIM_Cubic, leaveTangent=-1.3884, tangentMode=RCTM_User, time=0.6384, value=0.5912}
              {arriveTangent=-1.9181, tangentMode=RCTM_User, time=1, value=0}
             ]}}}))
        Passthrough Vector=<1 1 1>
        Spawn Group Mask=3
       )
       Spawn Group Mask=2
      )
      Spawn Group Mask=1
     )
    )) {module=ScaleMeshSize001}
   &ScaleMeshSize(Scale Factor=&VectorFromCurve(VectorCurve=DataInterface() {propertyValues {XCurve {keys [<at 0 1>,<at 1 0>]}, YCurve {keys [<at 0 1>,<at 1 0>]}, ZCurve {keys [<at 0 1>,<at 0.4606 0.7854>,<at 1 0>]}}})) {enabled=false}
   &SolveForcesAndVelocity(Clamp Velocity=false, Limit Acceleration=false)
   &Color(
    Color=<rgba 1 0.1018 0 1>
    Scale Alpha=&FloatFromCurve(FloatCurve=DataInterface() {propertyValues {Curve {keys [<at 0 1>,<at 0.25 1>,<at 0.8 1>,<at 1 1>]}}})
   ) {enabled=false}
   /SetVariables(
    Particles.BlackColor:LinearColor=<rgba 0 0 0 1>
    Particles.Int Fire:LinearColor=<rgba 0.848 2.348 4 1>
    Particles.LinesColor:LinearColor=&ColorFromCurve(ColorCurve=DataInterface() {propertyValues {
       AlphaCurve {keys [<at 0 0>,<at 0.0579 1>,<at 0.927 1 RCIM_Cubic>,<at 1 0 RCIM_Cubic>]}
       BlueCurve {keys [<at 1 0 RCIM_Cubic>]}
       GreenCurve {keys [<at 1 0 RCIM_Cubic>]}
       RedCurve {keys [<at 1 0 RCIM_Cubic>]}
      }})
    Particles.Ring Color:LinearColor=<rgba 3.3438 0 2.4158 1>
   ) {enabled=false, isSetParameters=true, module=SetVariables_#3}
   &DynamicMaterialParameters(
    Index 0 Param 1=0.0087
    Index 0 Param 3=-0.0477
    Module.Index0Float0Write=true
    Module.Index0Float1Write=true
    Module.Index0Float2Write=true
    Module.Index0Float3Write=true
   ) {enabled=false}
  ]
  rendererClasses [&NiagaraMeshRendererProperties, &NiagaraSpriteRendererProperties]
  renderers [
   {rendererClass=&NiagaraMeshRendererProperties, rendererIndex=0}
   {rendererClass=&NiagaraSpriteRendererProperties, rendererIndex=1}
  ]
 }
 Boost_Root001 {
  bEnabled=false
  emitterUpdateScript [
   &EmitterState(Inactive Response=::NewEnumerator2, Life Cycle Mode=::NewEnumerator1, Loop Duration=1, UseLoopDelay=false)
   &VC_FilteredSocketCount()
   &SpawnRate(SpawnRate=52, Use Spawn Probability=false) {enabled=false}
   &SpawnBurst_Instantaneous(Use Spawn Probability=false, UseLoopCountLimit=false)
   &SpawnRate(SpawnGroup=1, SpawnRate=52, Use Spawn Probability=false) {module=SpawnRate001}
   &SpawnRate(SpawnGroup=2, SpawnRate=32, Use Spawn Probability=false) {module=SpawnRate002}
   &SpawnRate(Spawn Probability=10, SpawnGroup=3, SpawnRate=11, Use Spawn Probability=true) {module=SpawnRate003}
  ]
  particleSpawnScript [
   /SetVariables(Particles.RenderTag:NiagaraInt32=&SpawnGroup()) {isSetParameters=true, module=SetVariables_#2}
   &VC_PaintedVariantsOverride() {enabled=false}
   &InitializeParticle(
    Color Mode=::NewEnumerator0
    Lifetime=&MaskFloatBySpawnGroup(
     Masked Float=1.25
     Passthrough Float=&MaskFloatBySpawnGroup(
      Masked Float=1
      Passthrough Float=&MaskFloatBySpawnGroup(
       Masked Float=1
       Passthrough Float=&MaskFloatBySpawnGroup(Masked Float=0.4, Passthrough Float=1, Spawn Group Mask=3)
       Spawn Group Mask=2
      )
      Spawn Group Mask=1
     )
    )
    Mesh Scale=&MaskVectorBySpawnGroup(
     Masked Vector=<0.2 0.2 0.2>
     Passthrough Vector=&MaskVectorBySpawnGroup(
      Masked Vector=<0.085 0.085 0.065>
      Passthrough Vector=&MaskVectorBySpawnGroup(
       Masked Vector=<0.2 0.2 0.2>
       Passthrough Vector=&MaskVectorBySpawnGroup(Masked Vector=<1 1 1>, Passthrough Vector=<1 1 1>, Spawn Group Mask=3)
       Spawn Group Mask=2
      )
      Spawn Group Mask=1
     )
    )
    Mesh Scale Mode=::NewEnumerator4
    Sprite Rotation Mode=::NewEnumerator2
    UsePositionOffset=false
    Write Material Random=true
   )
   &PartitionParticles(
    Particles Per Partition=0
    Partition Count Mode:ENiagara_InfiniteLoopDuration=::NewEnumerator1
    Partition Mode:ENiagara_PartitionMode=::NewEnumerator1
   )
   &InitialMeshOrientation(
    Mesh Axis To Orient=<0 0 0>
    Mesh Orientation Mode=::NewEnumerator2
    Orientation Vector=&MaskVectorBySpawnGroup(
     Masked Vector=<0 0 0>
     Passthrough Vector=&MaskVectorBySpawnGroup(
      Masked Vector=<0 0 0.5>
      Passthrough Vector=&MaskVectorBySpawnGroup(
       Masked Vector=<0 0 0.5>
       Passthrough Vector=&MaskVectorBySpawnGroup(Masked Vector=<0 0 0>, Passthrough Vector=<1 1 1>, Spawn Group Mask=3)
       Spawn Group Mask=2
      )
      Spawn Group Mask=1
     )
    )
    Rotation=&RandomRangeVector(
     Minimum=&MaskVectorBySpawnGroup(
      Masked Vector=<0 0 0>
      Passthrough Vector=&MaskVectorBySpawnGroup(
       Masked Vector=<0 0 -1>
       Passthrough Vector=&MaskVectorBySpawnGroup(
        Masked Vector=<0 0 -1>
        Passthrough Vector=&MaskVectorBySpawnGroup(Masked Vector=<-1 0 0>, Passthrough Vector=<1 1 1>, Spawn Group Mask=3)
        Spawn Group Mask=2
       )
       Spawn Group Mask=1
      )
     )
     Maximum=&MaskVectorBySpawnGroup(
      Masked Vector=<0 0 0>
      Passthrough Vector=&MaskVectorBySpawnGroup(
       Masked Vector=<0 0 1>
       Passthrough Vector=&MaskVectorBySpawnGroup(
        Masked Vector=<0 0 1>
        Passthrough Vector=&MaskVectorBySpawnGroup(Masked Vector=<1 0 0>, Passthrough Vector=<1 1 1>, Spawn Group Mask=3)
        Spawn Group Mask=2
       )
       Spawn Group Mask=1
      )
     )
    )
    Use Rotation Vector=true
   )
   &SkeletalMeshLocation(
    Bone / Socket Index=$Particles.PartitionParticles.Partition
    Bone / Socket Sampling Mode=::NewEnumerator5
    MaskSpawnGroup=false
    Mesh Sampling Type=::NewEnumerator6
    Sampled Position Offset=&AddVectorToPosition(
     Position=&MaskVectorBySpawnGroup(
      Masked Vector=<-232 0 85>
      Passthrough Vector=&MaskVectorBySpawnGroup(
       Masked Vector=<0 0 0>
       Passthrough Vector=&MaskVectorBySpawnGroup(
        Masked Vector=<0 0 0>
        Passthrough Vector=&MaskVectorBySpawnGroup(Masked Vector=<0 0 0>, Passthrough Vector=<1 1 1>, Spawn Group Mask=3)
        Spawn Group Mask=2
       )
       Spawn Group Mask=1
      )
     )
     Vector=Linked() {linkedVariable {name=User.DTCarOffset, type=NiagaraPosition}}
    )
    Skeletal Mesh=$System.SkeletalMesh
   )
   &AddVelocity(Velocity=&MaskVectorBySpawnGroup(Masked Vector=<0 0 0>, Passthrough Vector=<1 1 1>, Spawn Group Mask=3)) {enabled=false}
  ]
  particleUpdateScript [
   &ParticleState()
   &UpdateMeshOrientation(Rotation Vector=<0.5 0 0>)
   &ScaleMeshSize(Scale Factor=&MaskVectorBySpawnGroup(
     Masked Vector=&VectorFromCurve(VectorCurve=DataInterface() {propertyValues {XCurve {keys [<at 0 3>,<at 1 3>]}, YCurve {keys [<at 0 3>,<at 1 3>]}, ZCurve {keys [<at 0 3>,<at 1 3>]}}})
     Passthrough Vector=&MaskVectorBySpawnGroup(
      Masked Vector=&VectorFromCurve(VectorCurve=DataInterface() {propertyValues {
         XCurve {keys [<at 0.0085 0.7461>,<at 0.1197 0.9841>,<at 0.5195 0.898>,<at 0.9973 0.1466>]}
         YCurve {keys [<at 0.0017 0.7193>,<at 0.1289 0.9841>,<at 0.5226 0.9177>,<at 0.9973 0.1466>]}
         ZCurve {keys [<at 0.0022 0.7717>,<at 0.1517 1.1096>,<at 0.5124 1.0666>,<at 1.0156 0.1427>]}
        }})
      Passthrough Vector=&MaskVectorBySpawnGroup(
       Masked Vector=&VectorFromCurve(VectorCurve=DataInterface() {propertyValues {
          XCurve {keys [<at -0.0069 0.4922>,<at 0.0243 0.6397>,<at 0.495 0.6299>,<at 0.5 0.57>,<at 0.9519 0.57>,<at 1 0>]}
          YCurve {keys [<at -0.0069 0.4922>,<at 0.0377 0.6397>,<at 0.495 0.6299>,<at 0.5 0.57>,<at 0.9361 0.57>,<at 1 0>]}
          ZCurve {keys [<at 0 0.5668>,<at 0.0432 0.7017>,<at 0.495 0.7017>,<at 0.5 0.78>,<at 0.8305 0.78>,<at 1.0139 0.75>]}
         }})
       Passthrough Vector=&MaskVectorBySpawnGroup(
        Masked Vector=&VectorFromFloat(Value=&FloatFromCurve(FloatCurve=DataInterface() {propertyValues {Curve {keys [
              <at 0 0.2824>
              {arriveTangent=0.049, interpMode=RCIM_Cubic, leaveTangent=0.049, tangentMode=RCTM_User, time=0.2283, value=0.9882}
              {arriveTangent=-1.3884, interpMode=RCIM_Cubic, leaveTangent=-1.3884, tangentMode=RCTM_User, time=0.6384, value=0.5912}
              {arriveTangent=-1.9181, tangentMode=RCTM_User, time=1, value=0}
             ]}}}))
        Passthrough Vector=<1 1 1>
        Spawn Group Mask=3
       )
       Spawn Group Mask=2
      )
      Spawn Group Mask=1
     )
    )) {module=ScaleMeshSize001}
   &ScaleMeshSize(Scale Factor=&VectorFromCurve(VectorCurve=DataInterface() {propertyValues {XCurve {keys [<at 0 1>,<at 1 0>]}, YCurve {keys [<at 0 1>,<at 1 0>]}, ZCurve {keys [<at 0 1>,<at 0.4606 0.7854>,<at 1 0>]}}})) {enabled=false}
   &SolveForcesAndVelocity(Clamp Velocity=false, Limit Acceleration=false)
   &Color(
    Color=<rgba 0.379 0.462 0.5 1>
    Scale Alpha=&FloatFromCurve(FloatCurve=DataInterface() {propertyValues {Curve {keys [<at 0 0>,<at 0.25 1>,<at 0.8 1>,<at 1 0>]}}})
   )
   /SetVariables(
    Particles.BlackColor:LinearColor=<rgba 0 0 0 1>
    Particles.Int Fire:LinearColor=<rgba 0.848 2.348 4 1>
    Particles.LinesColor:LinearColor=&ColorFromCurve(ColorCurve=DataInterface() {propertyValues {
       AlphaCurve {keys [<at 0 0>,<at 0.0579 1>,<at 0.927 1 RCIM_Cubic>,<at 1 0 RCIM_Cubic>]}
       BlueCurve {keys [<at 1 0 RCIM_Cubic>]}
       GreenCurve {keys [<at 1 0 RCIM_Cubic>]}
       RedCurve {keys [<at 1 0 RCIM_Cubic>]}
      }})
    Particles.Ring Color:LinearColor=<rgba 3.3438 0 2.4158 1>
   ) {isSetParameters=true, module=SetVariables_#3}
   &DynamicMaterialParameters(
    Index 0 Param 1=0.0087
    Index 0 Param 2=-35.9908
    Index 0 Param 3=-0.0477
    Module.Index0Float0Write=true
    Module.Index0Float1Write=true
    Module.Index0Float2Write=true
    Module.Index0Float3Write=true
   )
  ]
  rendererClasses [&NiagaraMeshRendererProperties, &NiagaraSpriteRendererProperties]
  renderers [
   {rendererClass=&NiagaraMeshRendererProperties, rendererIndex=0}
   {rendererClass=&NiagaraMeshRendererProperties, rendererIndex=1}
   {rendererClass=&NiagaraMeshRendererProperties, rendererIndex=2}
   {rendererClass=&NiagaraMeshRendererProperties, rendererIndex=3}
   {rendererClass=&NiagaraSpriteRendererProperties, rendererIndex=4}
  ]
 }
 DiskCenter {
  emitterUpdateScript [
   &EmitterState(Inactive Response=::NewEnumerator2, Life Cycle Mode=::NewEnumerator1, Loop Duration=1, UseLoopDelay=false)
   &VC_FilteredSocketCount() {enabled=false}
   &SpawnRate(SpawnRate=52, Use Spawn Probability=false) {enabled=false}
   &SpawnBurst_Instantaneous(Use Spawn Probability=false, UseLoopCountLimit=false)
   &SpawnBurst_Instantaneous(Spawn Group=1, Use Spawn Probability=false, UseLoopCountLimit=false) {enabled=false, module=SpawnBurst_Instantaneous001}
   &SpawnRate(SpawnGroup=1, SpawnRate=52, Use Spawn Probability=false) {enabled=false, module=SpawnRate001}
  ]
  particleSpawnScript [
   /SetVariables(Particles.RenderTag:NiagaraInt32=&SpawnGroup()) {isSetParameters=true, module=SetVariables_#2}
   &VC_PaintedVariantsOverride()
   &InitializeParticle(
    Color Mode=::NewEnumerator0
    Lifetime=&MaskFloatBySpawnGroup(
     Masked Float=1
     Passthrough Float=&MaskFloatBySpawnGroup(
      Masked Float=1
      Passthrough Float=&MaskFloatBySpawnGroup(
       Masked Float=1
       Passthrough Float=&MaskFloatBySpawnGroup(Masked Float=1, Passthrough Float=1, Spawn Group Mask=3)
       Spawn Group Mask=2
      )
      Spawn Group Mask=1
     )
    )
    Mesh Scale=&MaskVectorBySpawnGroup(
     Masked Vector=<0.75 0.75 0.75>
     Passthrough Vector=&MaskVectorBySpawnGroup(
      Masked Vector=<0.2 0.2 0.2>
      Passthrough Vector=&MaskVectorBySpawnGroup(
       Masked Vector=<0.2 0.2 0.2>
       Passthrough Vector=&MaskVectorBySpawnGroup(Masked Vector=<1 1 1>, Passthrough Vector=<1 1 1>, Spawn Group Mask=3)
       Spawn Group Mask=2
      )
      Spawn Group Mask=1
     )
    )
    Mesh Scale Mode=::NewEnumerator4
    Position Offset=&AddVectorToPosition(
     Position=<-90 0 -42>
     Vector=&AddVectorToPosition(Position=$User.DTCarOffset, Vector=$User.SingleBoostPos)
    )
    Sprite Rotation Mode=::NewEnumerator2
    UsePositionOffset=true
    Write Material Random=true
   )
   &PartitionParticles(
    Distance=60
    Distance Falloff Exponent=100
    Number Of Partitions=$Emitter.PortCount
    Partition Mode:ENiagara_PartitionMode=::NewEnumerator3
    UseDistanceFalloff=true
   )
   &InitialMeshOrientation(
    Mesh Axis To Orient=<0 0 0>
    Mesh Orientation Mode=::NewEnumerator2
    Orientation Vector=&MaskVectorBySpawnGroup(
     Masked Vector=<0 0 0.5>
     Passthrough Vector=&MaskVectorBySpawnGroup(
      Masked Vector=<0 0 0.5>
      Passthrough Vector=&MaskVectorBySpawnGroup(
       Masked Vector=<0 0 0.5>
       Passthrough Vector=&MaskVectorBySpawnGroup(Masked Vector=<0 0 0>, Passthrough Vector=<1 1 1>, Spawn Group Mask=3)
       Spawn Group Mask=2
      )
      Spawn Group Mask=1
     )
    )
    Rotation=&RandomRangeVector(
     Minimum=&MaskVectorBySpawnGroup(
      Masked Vector=<0 0 -1>
      Passthrough Vector=&MaskVectorBySpawnGroup(
       Masked Vector=<0 0 -1>
       Passthrough Vector=&MaskVectorBySpawnGroup(
        Masked Vector=<0 0 -1>
        Passthrough Vector=&MaskVectorBySpawnGroup(Masked Vector=<-1 0 0>, Passthrough Vector=<1 1 1>, Spawn Group Mask=3)
        Spawn Group Mask=2
       )
       Spawn Group Mask=1
      )
     )
     Maximum=&MaskVectorBySpawnGroup(
      Masked Vector=<0 0 -1>
      Passthrough Vector=&MaskVectorBySpawnGroup(
       Masked Vector=<0 0 -1>
       Passthrough Vector=&MaskVectorBySpawnGroup(
        Masked Vector=<0 0 1>
        Passthrough Vector=&MaskVectorBySpawnGroup(Masked Vector=<1 0 0>, Passthrough Vector=<1 1 1>, Spawn Group Mask=3)
        Spawn Group Mask=2
       )
       Spawn Group Mask=1
      )
     )
    )
    Use Rotation Vector=true
   ) {enabled=false}
   &SkeletalMeshLocation(
    Bone / Socket Index=$Particles.PartitionParticles.Partition
    Bone / Socket Sampling Mode=::NewEnumerator5
    MaskSpawnGroup=false
    Mesh Sampling Type=::NewEnumerator6
    Sampled Position Offset=&AddVectorToPosition(
     Position=&MaskVectorBySpawnGroup(
      Masked Vector=<0 0 0>
      Passthrough Vector=&MaskVectorBySpawnGroup(
       Masked Vector=<-45 -31 5>
       Passthrough Vector=&MaskVectorBySpawnGroup(
        Masked Vector=<-25 -31 5>
        Passthrough Vector=&MaskVectorBySpawnGroup(Masked Vector=<0 0 0>, Passthrough Vector=<1 1 1>, Spawn Group Mask=3)
        Spawn Group Mask=2
       )
       Spawn Group Mask=1
      )
     )
     Vector=Linked() {linkedVariable {name=User.DTCarOffset, type=NiagaraPosition}}
    )
    Skeletal Mesh=$System.SkeletalMesh
   )
   &AddVelocity(Velocity=&MaskVectorBySpawnGroup(Masked Vector=<0 0 0>, Passthrough Vector=<1 1 1>, Spawn Group Mask=3)) {enabled=false}
  ]
  particleUpdateScript [
   &ParticleState()
   &UpdateMeshOrientation(Rotation Vector=&MaskVectorBySpawnGroup(Masked Vector=<1 0 0>, Passthrough Vector=<0 0 0>))
   &ScaleMeshSize(Scale Factor=&MaskVectorBySpawnGroup(
     Masked Vector=&VectorFromCurve(VectorCurve=DataInterface() {propertyValues {XCurve {keys [<at 0 1>,<at 1 1>]}, YCurve {keys [<at 0 1>,<at 1 1>]}, ZCurve {keys [<at 0 1>,<at 1 1>]}}})
     Passthrough Vector=&MaskVectorBySpawnGroup(
      Masked Vector=&VectorFromCurve(VectorCurve=DataInterface() {propertyValues {XCurve {keys [<at 0.0085 1>,<at 0.9973 1>]}, YCurve {keys [<at 0.0017 1>,<at 0.9973 1>]}, ZCurve {keys [<at 0.0022 1>,<at 1.0156 1>]}}})
      Passthrough Vector=&MaskVectorBySpawnGroup(
       Masked Vector=&VectorFromCurve(VectorCurve=DataInterface() {propertyValues {
          XCurve {keys [<at -0.0069 0.4922>,<at 0.0243 0.6397>,<at 0.495 0.6299>,<at 0.5 0.57>,<at 0.9519 0.57>,<at 1 0>]}
          YCurve {keys [<at -0.0069 0.4922>,<at 0.0377 0.6397>,<at 0.495 0.6299>,<at 0.5 0.57>,<at 0.9361 0.57>,<at 1 0>]}
          ZCurve {keys [<at 0 0.5668>,<at 0.0432 0.7017>,<at 0.495 0.7017>,<at 0.5 0.78>,<at 0.8305 0.78>,<at 1.0139 0.75>]}
         }})
       Passthrough Vector=&MaskVectorBySpawnGroup(
        Masked Vector=&VectorFromFloat(Value=&FloatFromCurve(FloatCurve=DataInterface() {propertyValues {Curve {keys [
              <at 0 0.2824>
              {arriveTangent=0.049, interpMode=RCIM_Cubic, leaveTangent=0.049, tangentMode=RCTM_User, time=0.2283, value=0.9882}
              {arriveTangent=-1.3884, interpMode=RCIM_Cubic, leaveTangent=-1.3884, tangentMode=RCTM_User, time=0.6384, value=0.5912}
              {arriveTangent=-1.9181, tangentMode=RCTM_User, time=1, value=0}
             ]}}}))
        Passthrough Vector=<1 1 1>
        Spawn Group Mask=3
       )
       Spawn Group Mask=2
      )
      Spawn Group Mask=1
     )
    )) {enabled=false, module=ScaleMeshSize001}
   &ScaleMeshSize(Scale Factor=&VectorFromCurve(VectorCurve=DataInterface() {propertyValues {XCurve {keys [<at 0 1>,<at 1 0>]}, YCurve {keys [<at 0 1>,<at 1 0>]}, ZCurve {keys [<at 0 1>,<at 0.4606 0.7854>,<at 1 0>]}}})) {enabled=false}
   &SolveForcesAndVelocity(Clamp Velocity=false, Limit Acceleration=false)
   &Color(
    Color=&ColorFromCurve(ColorCurve=DataInterface() {propertyValues {
       AlphaCurve {keys [<at 0 0.5>, {arriveTangent=0.7829, interpMode=RCIM_Cubic, leaveTangent=0.7829, time=0.4165, value=1}, <at 0.6387 1>, <at 1 0.5 RCIM_Cubic>]}
       BlueCurve {keys [<at 0.7691 1 RCIM_Cubic>]}
       GreenCurve {keys [<at 0.7691 0.8606 RCIM_Cubic>]}
       RedCurve {keys [<at 0.7691 0.586 RCIM_Cubic>]}
      }})
    Scale Alpha=&FloatFromCurve(FloatCurve=DataInterface() {propertyValues {Curve {keys [<at 0 0>,<at 0.25 1>,<at 0.8 1>,<at 1 0>]}}})
   )
   /SetVariables(
    Particles.BlackColor:LinearColor=<rgba 0.468 0.288 2 1>
    Particles.Int Fire:LinearColor=<rgba 0.848 2.348 4 1>
    Particles.LinesColor:LinearColor=&ColorFromCurve(ColorCurve=DataInterface() {propertyValues {
       AlphaCurve {keys [<at 0 0>,<at 0.0579 1>,<at 0.927 1 RCIM_Cubic>,<at 1 0 RCIM_Cubic>]}
       BlueCurve {keys [<at 1 1 RCIM_Cubic>]}
       GreenCurve {keys [<at 1 0.0358 RCIM_Cubic>]}
       RedCurve {keys [<at 1 0.1832 RCIM_Cubic>]}
      }})
    Particles.Ring Color:LinearColor=<rgba 3.3438 0 2.4158 1>
   ) {isSetParameters=true, module=SetVariables_#3}
   &DynamicMaterialParameters(
    Index 0 Param 1=0.0087
    Index 0 Param 2=-35.9908
    Index 0 Param 3=-0.0477
    Module.Index0Float0Write=true
    Module.Index0Float1Write=true
    Module.Index0Float2Write=true
    Module.Index0Float3Write=true
   )
  ]
  rendererClasses [&NiagaraMeshRendererProperties, &NiagaraSpriteRendererProperties]
  renderers [
   {rendererClass=&NiagaraMeshRendererProperties, rendererIndex=0}
   {rendererClass=&NiagaraMeshRendererProperties, rendererIndex=1}
   {rendererClass=&NiagaraMeshRendererProperties, rendererIndex=2}
   {rendererClass=&NiagaraMeshRendererProperties, rendererIndex=3}
   {rendererClass=&NiagaraSpriteRendererProperties, rendererIndex=4}
  ]
 }
 NE_BeeHive001 {
  emitterUpdateScript [
   &EmitterState(
    Inactive Response=::NewEnumerator1
    Life Cycle Mode=::NewEnumerator1
    Loop Behavior=::NewEnumerator1
    Loop Duration Mode=::NewEnumerator1
    UseLoopDelay=false
   )
   &SpawnBurst_Instantaneous(Use Spawn Probability=false, UseLoopCountLimit=false)
   &SpawnRate(SpawnGroup=1, SpawnRate=6, Use Spawn Probability=false) {enabled=false}
  ]
  particleSpawnScript [
   &VC_PaintedVariantsOverride(
    Black=<rgba 0.0194 0.0194 0.02 1>
    BurntSienna=<rgba 0.14 0.0257 0.0028 1>
    Cobalt=<rgba 0.0198 0.1192 1.3 1>
    Crimson=<rgba 0.5 0.005 0.0132 1>
    DefaultColor=<rgba 1 0.2 0 1>
    ForestGreen=<rgba 0 0.4 0 1>
    Grey=<rgba 0.17 0.17 0.17 1>
    Lime=<rgba 0.2475 0.45 0 1>
    Orange=<rgba 0.7 0.0817 0 1>
    Pink=<rgba 0.643 0.1714 0.7 1>
    Purple=<rgba 0.1066 0.052 0.52 1>
    Saffron=<rgba 0.45 0.375 0 1>
    SkyBlue=<rgba 0 0.3333 1 1>
    TitaniumWhite=<rgba 0.4 0.4 0.4 1>
   )
   &InitializeParticle(
    Color=$Particles.Color
    Lifetime=&MaskFloatBySpawnGroup(
     Masked Float=1
     Passthrough Float=&MaskFloatBySpawnGroup(
      Masked Float=1
      Passthrough Float=&MaskFloatBySpawnGroup(
       Masked Float=1
       Passthrough Float=&MaskFloatBySpawnGroup(Masked Float=1, Passthrough Float=1, Spawn Group Mask=3)
       Spawn Group Mask=2
      )
      Spawn Group Mask=1
     )
    )
    Mesh Scale=&MaskVectorBySpawnGroup(
     Masked Vector=<0.5 0.5 0.5>
     Passthrough Vector=&MaskVectorBySpawnGroup(
      Masked Vector=<0.2 0.2 0.2>
      Passthrough Vector=&MaskVectorBySpawnGroup(
       Masked Vector=<0.2 0.2 0.2>
       Passthrough Vector=&MaskVectorBySpawnGroup(Masked Vector=<1 1 1>, Passthrough Vector=<1 1 1>, Spawn Group Mask=3)
       Spawn Group Mask=2
      )
      Spawn Group Mask=1
     )
    )
    Mesh Scale Mode=::NewEnumerator4
    Position Offset=&AddVectorToPosition(
     Position=<-90 0 -42>
     Vector=&AddVectorToPosition(Position=$User.DTCarOffset, Vector=$User.SingleBoostPos)
    )
    Position Offset Coordinate Space=::Local
    Sprite Rotation Mode=::NewEnumerator2
    Sprite Size Mode=::NewEnumerator3
    UsePositionOffset=true
    Write Material Random=true
   )
   &AddVelocity(Velocity=<0 0 0>)
  ]
  particleUpdateScript [
   &ParticleState(Kill Particles When Lifetime Has Elapsed=false, Let Infinitely Lived Particles Die When Emitter Deactivates=true)
   &UpdateMeshOrientation(Rotation Vector=&MaskVectorBySpawnGroup(Masked Vector=<1 0 0>, Passthrough Vector=<-1 0 0>))
   &SolveForcesAndVelocity(Clamp Velocity=false, Limit Acceleration=false)
   /SetVariables(
    Particles.BlackColor:LinearColor=<rgba 0.468 0.288 2 1>
    Particles.LinesColor:LinearColor=&ColorFromCurve(ColorCurve=DataInterface() {propertyValues {
       AlphaCurve {keys [<at 0 0>,<at 0.0579 1>,<at 0.927 1 RCIM_Cubic>,<at 1 0 RCIM_Cubic>]}
       BlueCurve {keys [<at 1 1 RCIM_Cubic>]}
       GreenCurve {keys [<at 1 0.0358 RCIM_Cubic>]}
       RedCurve {keys [<at 1 0.1832 RCIM_Cubic>]}
      }})
    Particles.Rings Base:LinearColor=<rgba 0.848 2.348 4 1>
    Particles.Rings Fill:LinearColor=<rgba 4 1.8347 0.848 1>
   ) {isSetParameters=true}
   &ScaleColor(Scale RGB=<2 2 2>, ScaleA=true, ScaleRGB=true) {enabled=false}
   &ShapeLocation(
    MaskSpawnGroup=false
    Offset=&MakeVector(
     X=&Sine(Normalized Angle=$Emitter.Age, Period=0.6, Scale=1, Bias=0)
     Y=&Cosine(Normalized Angle=$Emitter.Age, Period=0.6, Scale=1, Bias=0)
     Z=&Sine(Normalized Angle=$Emitter.Age, Period=0.25, Scale=1, Bias=0)
    )
    Offset Mode=::NewEnumerator0
    Override Seed=false
    Shape Origin=$Particles.Initial.Position
    Sphere Radius=0
    UseOwnerScale=true
   ) {enabled=false}
  ]
  rendererClasses [&NiagaraSpriteRendererProperties, &NiagaraMeshRendererProperties]
  renderers [
   {rendererClass=&NiagaraSpriteRendererProperties}
   {rendererClass=&NiagaraMeshRendererProperties, rendererIndex=1}
   {rendererClass=&NiagaraMeshRendererProperties, rendererIndex=2}
   {rendererClass=&NiagaraMeshRendererProperties, rendererIndex=3}
  ]
 }
 NE_BlackFlame_Base1 {
  bEnabled=false
  emitterUpdateScript [
   &EmitterState(Life Cycle Mode=::NewEnumerator1, Loop Duration=1, UseLoopDelay=false)
   &VC_FilteredSocketCount()
   &VC_RateQualityHandler(Epic Spawn Rate=30, High Spawn Rate=30, Low Spawn Rate=30, Use Custom Rate?=true)
   &VC_SpawnRibbon_v1(Number of Ribbons=$Emitter.PortCount, Spawn Time=-1, SpawnRate=$Emitter.Spawn Rate)
  ]
  particleSpawnScript [
   &VC_PaintedVariantsOverride() {enabled=false}
   &VC_InitializeRibbon_v1(
    Color=$Particles.Color
    Custom Offset=true
    Enable Skeletal Position=true
    Lifetime=&MergeHelper_FloatSwitch(0=1, 1=1, 2=1, 3=?, Index=$User.GameModeID, Num=::NewEnumerator1)
    Local Vector Offset=&AddVectorToPosition(Position=<-15.4817 0 0>, Vector=Linked() {linkedVariable {name=User.DTCarOffset003, type=NiagaraPosition}})
    Offset Amount=1
    Ribbon Width=60
   )
   &AddVelocity(Velocity=<-1750 0 0>) {module=AddVelocity001}
   /SetVariables(
    Particles.R1_Width:NiagaraFloat=&SelectFloatFromArray(Direct Array Index=&SpawnGroup(), Float Selection Array=DataInterface() {propertyValues {FloatData [75,0]}})
    Particles.R2_Width:NiagaraFloat=&SelectFloatFromArray(Direct Array Index=&SpawnGroup(), Float Selection Array=DataInterface() {propertyValues {FloatData [0,-105]}})
    Particles.R3_Color:LinearColor=<rgba 0.2241 0.5211 0.6609 3>
   ) {enabled=false, isSetParameters=true}
  ]
  particleUpdateScript [
   &ParticleState()
   &InheritSourceMovement(
    Applied Position Scale=&Multiply_VectorByFloat(
     Vector=<1 1 1>
     Float=&FloatFromCurve(FloatCurve=DataInterface() {propertyValues {Curve {keys [{interpMode=RCIM_Cubic, tangentMode=RCTM_Break, time=0, value=1}, {interpMode=RCIM_Cubic, tangentMode=RCTM_Break, time=1, value=0}]}}})
    )
    ApplyForce=false
    ApplyPosition=true
    ApplyPositionFalloff=false
   ) {module=InheritSourceMovement001}
   &SolveForcesAndVelocity(Clamp Velocity=false, Limit Acceleration=false)
   &VC_UpdateRibbon_v1(Enable Skeletal Position=true)
   /SetVariables(Particles.DotProduct:NiagaraFloat=&DotProduct(
     Vector A=/Niagara/DynamicInputs/Math/TransformVectorByMatrix(
      Vector:Vector3f=<1 0 0>
      Transformation Matrix:NiagaraMatrix=$Engine.Owner.SystemLocalToWorld
      Transform Vector as Position:NiagaraBool=false
     )
     Vector B=&NormalizeVector(
      Vector To Normalize=&SubtractVectorFromPosition(
       Position=$Engine.Owner.Position
       Vector=&ReturnCameraPosition(Coordinate Space:ENiagaraCoordinateSpace=::Simulation, Camera Query:NiagaraDataInterfaceCamera=DataInterface())
      )
      Safe Normalize=true
      Safe Fallback Threshold=0
      Safe Fallback Vector=<0 0 0.001>
     )
    )) {isSetParameters=true}
   /SetVariables(
    Particles.RibbonBack:LinearColor=&ColorFromCurve(ColorCurve=DataInterface() {propertyValues {
       AlphaCurve {keys [<at 0.5096 0>,<at 0.6208 2>,<at 0.697 0>]}
       BlueCurve {keys [<at 0.9243 0>]}
       GreenCurve {keys [<at 0.9243 0>]}
       RedCurve {keys [<at 0.9243 0>]}
      }})
    Particles.RibbonBase:LinearColor=&ColorFromCurve(ColorCurve=DataInterface() {propertyValues {
       AlphaCurve {keys [<at 0.1162 0>,<at 0.1621 1>,<at 0.31 2>,<at 0.4762 0>]}
       BlueCurve {keys [<at 0.9243 0>]}
       GreenCurve {keys [<at 0.9243 0>]}
       RedCurve {keys [<at 0.9243 0>]}
      }})
    Particles.RibbonGlow:LinearColor=&ColorFromCurve(ColorCurve=DataInterface() {propertyValues {
       AlphaCurve {keys [<at 0.2665 3>,<at 0.3905 0>,<at 0.5841 0>]}
       BlueCurve {keys [<at 0.6966 2>]}
       GreenCurve {keys [<at 0.6966 0.5201>]}
       RedCurve {keys [<at 0.6966 0>]}
      }})
    Particles.RibbonGradient:LinearColor=&ColorFromCurve(ColorCurve=DataInterface() {propertyValues {
       AlphaCurve {keys [
         <at 0.2819 7>
         <at 0.4191 0>
         <at 0.6302 0>
         <at 0.7212 1>
         {arriveTangent=-6.7757, interpMode=RCIM_Cubic, leaveTangent=-6.7757, time=0.7822, value=1}
         <at 0.8688 0>
        ]}
       BlueCurve {keys [<at 1 2 RCIM_Cubic>]}
       GreenCurve {keys [<at 1 0.2817 RCIM_Cubic>]}
       RedCurve {keys [<at 1 0 RCIM_Cubic>]}
      }})
   ) {isSetParameters=true}
   &ScaleRibbonWidth(Ribbon Width Scale=&FloatFromCurve(FloatCurve=DataInterface() {propertyValues {Curve {keys [
         {arriveTangent=-0.2605, interpMode=RCIM_Cubic, leaveTangent=-0.2605, tangentMode=RCTM_User, time=0.0043, value=1.4579}
         {arriveTangent=-1.5709, interpMode=RCIM_Cubic, leaveTangent=-1.5709, tangentMode=RCTM_User, time=0.3137, value=1.1734}
         {arriveTangent=-2.7437, interpMode=RCIM_Cubic, leaveTangent=-2.7437, time=0.5043, value=0.5604}
         {arriveTangent=-1.126, interpMode=RCIM_Cubic, leaveTangent=-1.126, time=0.6133, value=0.3515}
         {arriveTangent=-1.0469, tangentMode=RCTM_User, time=1.002, value=0}
        ]}}}))
   /SetVariables(
    Particles.Ribbon_BlackColor:LinearColor=<rgba 0 0 0 1>
    Particles.WaveScaleCurve:NiagaraFloat=&FloatFromCurve(
     FloatCurve=DataInterface() {propertyValues {Curve {keys [
         {arriveTangent=39.5227, interpMode=RCIM_Cubic, leaveTangent=39.5227, tangentMode=RCTM_User, time=0.0092, value=0}
         {
          arriveTangent=0.7219
          arriveTangentWeight=0.0795
          interpMode=RCIM_Cubic
          leaveTangent=0.1806
          leaveTangentWeight=0.0696
          tangentMode=RCTM_Break
          tangentWeightMode=RCTWM_WeightedBoth
          time=0.125
          value=1.25
         }
         {arriveTangent=0.1988, interpMode=RCIM_Cubic, leaveTangent=0.1988, tangentMode=RCTM_User, time=0.45, value=0.625}
        ]}}}
     Scale Curve=10
    )
   ) {isSetParameters=true}
   /SetVariables(
    Particles.CustomPosition_R1:NiagaraPosition=&AddVectorToPosition(
     Position=$Particles.Position
     Vector=&Multiply_VectorByFloat(
      Vector=&MakeVector(
       X=&Sine(
        Normalized Angle=&Add_Float(A=0, B=&Subtract_Float(A=$Emitter.NormalizedLoopAge, B=$Particles.NormalizedAge))
        Period=$User.WavePeriod
        Scale=&Multiply_Float(
         A=&SelectFloatFromArray(Direct Array Index=$Particles.SpawnGroup, Float Selection Array=DataInterface() {propertyValues {FloatData [1,-1]}})
         B=$User.WaveMaxScale
        )
        Bias=0
       )
       Y=&Sine(
        Normalized Angle=&Add_Float(A=0, B=&Subtract_Float(A=$Emitter.NormalizedLoopAge, B=$Particles.NormalizedAge))
        Period=$User.WavePeriod
        Scale=&Multiply_Float(
         A=&SelectFloatFromArray(Direct Array Index=$Particles.SpawnGroup, Float Selection Array=DataInterface() {propertyValues {FloatData [1,-1]}})
         B=$User.WaveMaxScale
        )
        Bias=0
       )
       Z=&Cosine(
        Normalized Angle=&Add_Float(A=0, B=&Subtract_Float(A=$Emitter.NormalizedLoopAge, B=$Particles.NormalizedAge))
        Period=$User.WavePeriod
        Scale=$User.WaveMaxScale
        Bias=0
       )
      )
      Float=$Particles.WaveScaleCurve
     )
    )
    Particles.CustomPosition_R2:NiagaraPosition=&AddVectorToPosition(
     Position=$Particles.Position
     Vector=&Multiply_VectorByFloat(
      Vector=&MakeVector(
       X=&Sine(
        Normalized Angle=&Add_Float(
         A=&Multiply_Float(A=$User.WavePeriod, B=0.5)
         B=&Subtract_Float(A=$Emitter.NormalizedLoopAge, B=$Particles.NormalizedAge)
        )
        Period=$User.WavePeriod
        Scale=&Multiply_Float(
         A=&SelectFloatFromArray(Direct Array Index=$Particles.SpawnGroup, Float Selection Array=DataInterface() {propertyValues {FloatData [1,-1]}})
         B=$User.WaveMaxScale
        )
        Bias=0
       )
       Y=&Sine(
        Normalized Angle=&Add_Float(
         A=&Multiply_Float(A=$User.WavePeriod, B=0.5)
         B=&Subtract_Float(A=$Emitter.NormalizedLoopAge, B=$Particles.NormalizedAge)
        )
        Period=$User.WavePeriod
        Scale=&Multiply_Float(
         A=&SelectFloatFromArray(Direct Array Index=$Particles.SpawnGroup, Float Selection Array=DataInterface() {propertyValues {FloatData [1,-1]}})
         B=$User.WaveMaxScale
        )
        Bias=0
       )
       Z=&Cosine(
        Normalized Angle=&Add_Float(
         A=&Multiply_Float(A=$User.WavePeriod, B=0.5)
         B=&Subtract_Float(A=$Emitter.NormalizedLoopAge, B=$Particles.NormalizedAge)
        )
        Period=$User.WavePeriod
        Scale=$User.WaveMaxScale
        Bias=0
       )
      )
      Float=$Particles.WaveScaleCurve
     )
    )
   ) {isSetParameters=true}
   /SetVariables(Particles.DynamicMP_R2:Vector4f=&MakeVector4(
     X=&FloatFromCurve(FloatCurve=DataInterface() {propertyValues {Curve {
         defaultValue=1
         keys [
          {tangentMode=RCTM_User, time=0, value=0.5}
          {arriveTangent=-0.0633, interpMode=RCIM_Cubic, leaveTangent=-0.0633, tangentMode=RCTM_User, time=0.65, value=0.5}
          {arriveTangent=-0.0699, interpMode=RCIM_Cubic, leaveTangent=-0.0699, tangentMode=RCTM_User, time=0.75, value=-2}
         ]
        }}})
     Y=0
     Z=0
     W=0
    )) {isSetParameters=true}
   &DynamicMaterialParameters(
    Index 0 Param 1=&FloatFromCurve(FloatCurve=DataInterface() {propertyValues {Curve {
        defaultValue=1
        keys [
         {
          arriveTangent=0.0405
          arriveTangentWeight=0.1835
          interpMode=RCIM_Cubic
          leaveTangent=0.0405
          leaveTangentWeight=0.1835
          tangentMode=RCTM_User
          time=0
          value=0
         }
         {
          arriveTangent=0.0201
          arriveTangentWeight=0.1834
          interpMode=RCIM_Cubic
          leaveTangent=0.0201
          leaveTangentWeight=0.1834
          tangentMode=RCTM_User
          time=0.5
          value=1
         }
        ]
       }}})
    Module.Index0Float0Write=true
    Module.Index0Float1Write=true
    Module.Index0Float2Write=true
    Module.Index0Float3Write=true
   )
  ]
  rendererClasses [&NiagaraRibbonRendererProperties]
  renderers [
   {rendererClass=&NiagaraRibbonRendererProperties}
   {rendererClass=&NiagaraRibbonRendererProperties, rendererIndex=1}
   {rendererClass=&NiagaraRibbonRendererProperties, rendererIndex=2}
   {rendererClass=&NiagaraRibbonRendererProperties, rendererIndex=3}
   {rendererClass=&NiagaraRibbonRendererProperties, rendererIndex=4}
  ]
 }
 NE_VC_Boost_Ribbon_Parent001 {
  emitterUpdateScript [
   &EmitterState()
   &VC_FilteredSocketCount()
   &VC_RateQualityHandler()
   &VC_SpawnRibbon_v1(Number of Ribbons=$Emitter.PortCount, SpawnRate=50)
  ]
  particleSpawnScript [
   &VC_PaintedVariantsOverride(Purple=<rgba 0.2789 0 0.6795 1>, R1 Orange:LinearColor=<rgba 0.6724 0.2346 0 1>, Use R1 Painted Overrides=true)
   &VC_InitializeRibbon_v1(
    Custom Offset=true
    Lifetime=0.6
    Local Vector Offset=&AddVectorToPosition(
     Position=<-90 0 0>
     Vector=&AddVectorToPosition(Position=$User.DTCarOffset, Vector=Linked() {linkedVariable {name=User.SingleBoostPos, type=NiagaraPosition}})
    )
    Offset Amount=1
    Ribbon Width:NiagaraFloat=&Multiply_Float(A=65, B=1.4)
   )
   &Color(
    Color=&MakeCustomLinearColorFromBool(Bool=$User.bIsPainted, True Color=$User.PaintedColor, False Color=<rgba 1 1 1 1>)
    Scale Alpha=&MakeCustomFloatFromBool(Bool=$User.bIsPainted, True Float=1.5, False Float=1)
   )
   &AddVelocity(Velocity=&MakeCustomVectorFromBool(Bool=$User.bIsFrontendPreview, True Vector=<-2000 0 0>, False Vector=<-2000 0 0>)) {module=AddVelocity001}
   /SetVariables(
    Particles.COLOR2:LinearColor=&MakeCustomLinearColorFromBool(Bool=$User.bIsPainted, True Color=$User.PaintedColor, False Color=<rgba 1 1 1 1>)
    Particles.WIDTH_2=180
   ) {isSetParameters=true}
  ]
  particleUpdateScript [
   &ParticleState()
   &InheritSourceMovement(
    Applied Position Scale=&Multiply_VectorByFloat(
     Vector=<1 1 1>
     Float=&FloatFromCurve(FloatCurve=DataInterface() {propertyValues {Curve {keys [<at 0 1 RCIM_Cubic>, {interpMode=RCIM_Cubic, leaveTangent=-3.4326, time=0.7, value=0}]}}})
    )
    ApplyForce=false
    ApplyPosition=true
    ApplyPositionFalloff=false
   ) {module=InheritSourceMovement001}
   &SolveForcesAndVelocity(Clamp Velocity=false, Limit Acceleration=false)
   &VC_UpdateRibbon_v1()
   &DynamicMaterialParameters(
    Index 0 Param 1=&MakeCustomFloatFromBool(Bool=$User.bIsPainted, True Float=5, False Float=0)
    Index 2 Param 1=5
    Module.Index0Param0Write=true
    Module.Index0Param1Write=true
    Module.Index0Param2Write=true
    Module.Index0Param3Write=true
    Module.Index2Param0Write=true
    Module.Index2Param1Write=true
    Module.Index2Param2Write=true
    Module.Index2Param3Write=true
    Write Parameter Index 2=true
   )
   &ScaleRibbonWidth(Ribbon Width Scale=&FloatFromCurve(FloatCurve=DataInterface() {propertyValues {Curve {keys [
         <at 0 1.5 RCIM_Cubic>
         {arriveTangent=-0.293, interpMode=RCIM_Cubic, leaveTangent=0.0883, tangentMode=RCTM_Break, time=0.2, value=0.6}
         <at 1 0.3 RCIM_Cubic>
        ]}}}))
   &ScaleColor(
    Scale RGB=&VectorFromFloat(Value=&MakeCustomFloatFromBool(Bool=$User.bIsPainted, True Float=$User.PaintedScalar, False Float=1))
    ScaleA=true
    ScaleRGB=true
   )
   /SetVariables(
    Particles.INDEX0:Vector4f=&MakeVector4(
     X=&MakeCustomFloatFromBool(Bool=$User.bIsPainted, True Float=5, False Float=0)
     Y=&MakeCustomFloatFromBool(Bool=$User.bIsPainted, True Float=$User.HighlightPowerTrail, False Float=2)
     Z=&MakeCustomFloatFromBool(Bool=$User.bIsPainted, True Float=$User.MultiVariantColorTrail, False Float=1)
     W=&MakeCustomFloatFromBool(Bool=$User.bIsPainted, True Float=$User.WhiteColorTrail, False Float=1)
    )
    Particles.INDEX1:Vector4f=&MakeVector4_FromLinearColor(LinearColor=$User.PaintedColor)
    Particles.INDEX2:Vector4f=&MakeVector4(X=0, Y=0.1, Z=1.5, W=0)
    Particles.INDEX3:Vector4f=&MakeVector4(X=0, Y=0, Z=0, W=0)
   ) {isSetParameters=true}
   /SetVariables(
    Particles.SCALECOLOR2:LinearColor=&VC_DynamicScaleColor(
     Color to Scale:LinearColor=$Particles.COLOR2
     RGB Scale:Vector3f=&VectorFromFloat(Value=&MakeCustomFloatFromBool(
       Bool=$User.bIsPainted
       True Float=&MakeCustomFloatFromBool(
        Bool=$User.bIsFrontendPreview
        True Float=&Multiply_Float(A=$User.MultiVariantColorTrail, B=2)
        False Float=&Multiply_Float(A=$User.MultiVariantColorTrail, B=1)
       )
       False Float=1
      ))
     Alpha Scale:NiagaraFloat=&FloatFromCurve(FloatCurve=DataInterface() {propertyValues {Curve {keys [
          {interpMode=RCIM_Cubic, tangentMode=RCTM_Break, time=0.01, value=0}
          {arriveTangent=3.5161, interpMode=RCIM_Cubic, tangentMode=RCTM_Break, time=0.18, value=1}
          {interpMode=RCIM_Cubic, tangentMode=RCTM_Break, time=1, value=0}
         ]}}})
    )
    Particles.WIDTH_2:NiagaraFloat=&FloatFromCurve(
     FloatCurve=DataInterface() {propertyValues {Curve {keys [
         <at 0 1.5 RCIM_Cubic>
         {arriveTangent=-0.293, interpMode=RCIM_Cubic, leaveTangent=0.0883, tangentMode=RCTM_Break, time=0.2, value=0.6}
         <at 1 0.3 RCIM_Cubic>
        ]}}}
     Scale Curve=180
    )
   ) {isSetParameters=true}
  ]
  rendererClasses [&NiagaraRibbonRendererProperties]
  renderers [
   {rendererClass=&NiagaraRibbonRendererProperties}
   {rendererClass=&NiagaraRibbonRendererProperties, rendererIndex=1}
  ]
 }
 NE_VC_Boost_Sprite_Parent {
  emitterUpdateScript [
   &EmitterState(
    Inactive Response=::NewEnumerator1
    Life Cycle Mode=::NewEnumerator1
    Loop Behavior=::NewEnumerator1
    Loop Duration Mode=::NewEnumerator1
    UseLoopDelay=false
   )
   &VC_FilteredSocketCount() {enabled=false}
   &SpawnBurst_Instantaneous(Use Spawn Probability=false, UseLoopCountLimit=false) {module=SpawnBurst_Instantaneous001}
   &SpawnRate(SpawnRate=50, Use Spawn Probability=false) {enabled=false}
  ]
  particleSpawnScript [
   &VC_PaintedVariantsOverride(Purple=<rgba 0.2789 0 0.6795 1>, Use R1 Painted Overrides=true)
   &InitializeParticle(
    Color Mode=::NewEnumerator0
    Lifetime=$Emitter.CurrentLoopDuration
    Mesh Scale Mode=::NewEnumerator3
    Mesh Uniform Scale=&Multiply_Float(A=0.15, B=2.8)
    Position Offset=&AddVectorToPosition(
     Position=<-110 0 0>
     Vector=&AddVectorToPosition(Position=$User.SingleBoostPos001, Vector=Linked() {linkedVariable {name=User.DTCarOffset, type=NiagaraPosition}})
    )
    Position Offset Coordinate Space=::Local
    Sprite Rotation Mode=::NewEnumerator2
    UsePositionOffset=true
    Write Material Random=true
   )
   &PartitionParticles(Number Of Partitions=$Emitter.PortCount) {enabled=false}
   &SkeletalMeshLocation(
    Bone / Socket Index=$Particles.PartitionParticles.Partition
    Bone / Socket Sampling Mode=::NewEnumerator5
    MaskSpawnGroup=false
    Mesh Sampling Type=::NewEnumerator6
    Skeletal Mesh=$System.SkeletalMesh
   ) {enabled=false}
   &AddVelocity(Velocity=<-1500 0 0>) {enabled=false}
   &Color(Color=&MakeCustomLinearColorFromBool(Bool=$User.bIsPainted, True Color=$User.PaintedColor, False Color=<rgba 1 1 1 1>))
   /SetVariables(Particles.COLOR2:LinearColor=&MakeCustomLinearColorFromBool(Bool=$User.bIsPainted, True Color=$User.PaintedColor, False Color=<rgba 1 1 1 1>)) {isSetParameters=true}
  ]
  particleUpdateScript [
   &ParticleState()
   &ScaleMeshSize(Scale Factor=&VectorFromFloat(Value=&FloatFromCurve(
      FloatCurve=DataInterface() {propertyValues {Curve {keys [
          {interpMode=RCIM_Cubic, tangentMode=RCTM_Break, time=0, value=1}
          {interpMode=RCIM_Cubic, tangentMode=RCTM_Break, time=0.5, value=0.85}
          {interpMode=RCIM_Cubic, tangentMode=RCTM_Break, time=1, value=1}
         ]}}}
      Scale Curve=1.2
     )))
   &SolveForcesAndVelocity(Clamp Velocity=false, Limit Acceleration=false)
   &DynamicMaterialParameters(
    Index 0 Param 1=&MakeCustomFloatFromBool(Bool=$User.bIsPainted, True Float=1, False Float=0)
    Index 0 Param 2=0.8
    Index 2 Param 1=&MakeCustomFloatFromBool(Bool=$User.bIsPainted, True Float=1, False Float=0)
    Index 2 Param 2=1
    Module.Index0Param0Write=true
    Module.Index0Param1Write=true
    Module.Index0Param2Write=true
    Module.Index0Param3Write=true
    Module.Index1Param0Write=true
    Module.Index1Param1Write=true
    Module.Index1Param2Write=true
    Module.Index1Param3Write=true
    Module.Index2Param0Write=true
    Module.Index2Param1Write=true
    Module.Index2Param2Write=true
    Module.Index2Param3Write=true
    Module.Index3Param0Write=true
    Module.Index3Param1Write=true
    Module.Index3Param2Write=true
    Module.Index3Param3Write=true
    Write Parameter Index 1=true
    Write Parameter Index 2=true
    Write Parameter Index 3=true
   )
   &ScaleColor(
    Scale RGB=&VectorFromFloat(Value=&MakeCustomFloatFromBool(Bool=$User.bIsPainted, True Float=$User.PaintedScalar, False Float=1))
    ScaleA=true
    ScaleRGB=true
   )
   /SetVariables(Particles.SCALE_COLOR2:LinearColor=&VC_DynamicScaleColor(
     Color to Scale:LinearColor=$Particles.COLOR2
     RGB Scale:Vector3f=&VectorFromFloat(Value=&MakeCustomFloatFromBool(Bool=$User.bIsPainted, True Float=&Multiply_Float(A=$User.MultiVariantColorCone, B=1.2), False Float=1))
     Alpha Scale:NiagaraFloat=1
    )) {isSetParameters=true}
  ]
  rendererClasses [&NiagaraMeshRendererProperties, &NiagaraSpriteRendererProperties]
  renderers [
   {rendererClass=&NiagaraMeshRendererProperties, rendererIndex=0}
   {rendererClass=&NiagaraMeshRendererProperties, rendererIndex=1}
   {rendererClass=&NiagaraSpriteRendererProperties, rendererIndex=2}
  ]
 }
 Ribbon_HaloSpin {
  emitterUpdateScript [
   &EmitterState(Inactive Response=::NewEnumerator1, Life Cycle Mode=::NewEnumerator1, Loop Duration=1.1, UseLoopDelay=false)
   &VC_FilteredSocketCount()
   &VC_RateQualityHandler(Epic Spawn Rate=30, High Spawn Rate=30, Low Spawn Rate=30, Use Custom Rate?=true)
   &VC_SpawnRibbon_v1(Number of Ribbons=$Emitter.PortCount, Spawn Mode:ENiagaraSkeletalSpawning=::NewEnumerator0, Spawn Time=-1, SpawnRate=$Emitter.Spawn Rate)
  ]
  particleSpawnScript [
   &VC_PaintedVariantsOverride(
    Black=<rgba 0.01 0.01 0.01 1.25>
    BurntSienna=<rgba 0.0955 0.0295 0.0118 1>
    Cobalt=<rgba 0 0.0066 0.3559 1>
    Crimson=<rgba 0.0816 0 0.0022 1.5>
    ForestGreen=<rgba 0 0.0608 0.0031 1.5>
    Grey=<rgba 0.1649 0.1649 0.1649 0.4435>
    Lime=<rgba 0.0334 0.1565 0.025 1.5>
    Orange=<rgba 0.2205 0.0372 0 1.5>
    Pink=<rgba 0.3003 0.0771 0.2854 1>
    Purple=<rgba 0.0223 0 0.1101 2>
    R1 Black:LinearColor=<rgba 0.0469 0.0469 0.0469 1>
    R1 Cobalt:LinearColor=<rgba 0 0.0185 1 1>
    R1 ForestGreen:LinearColor=<rgba 0 0.2319 0.004 1>
    R1 Grey:LinearColor=<rgba 0.1267 0.1267 0.1267 1>
    R1 Orange:LinearColor=<rgba 0.6724 0.2346 0 1>
    R1 Pink:LinearColor=<rgba 0.5052 0.1552 0.3779 1>
    R1 Purple:LinearColor=<rgba 0.0699 0 0.2552 1>
    R1 SkyBlue:LinearColor=<rgba 0.2058 0.4258 0.5365 1>
    Saffron=<rgba 0.1372 0.1052 0 1>
    SkyBlue=<rgba 0 0.0765 0.1719 1.5>
    TitaniumWhite=<rgba 0.1406 0.1406 0.1406 1>
    Use R1 Painted Overrides=true
   )
   &VC_InitializeRibbon_v1(
    Color=$Particles.Color
    Custom Offset=true
    Lifetime=&MergeHelper_FloatSwitch(0=1, 1=1, 2=1, 3=?, Index=$User.GameModeID, Num=::NewEnumerator1)
    Local Vector Offset=&AddVectorToPosition(Position=<-168.1301 0 35>, Vector=<0 0 0>)
    Offset Amount=1
    Ribbon Width=20
   )
   &AddVelocity(Velocity=<-1750 0 0>) {module=AddVelocity001}
  ]
  particleUpdateScript [
   &ParticleState()
   &InheritSourceMovement(
    Applied Position Scale=&Multiply_VectorByFloat(
     Vector=<1 1 1>
     Float=&FloatFromCurve(FloatCurve=DataInterface() {propertyValues {Curve {keys [{interpMode=RCIM_Cubic, tangentMode=RCTM_Break, time=0, value=1}, {interpMode=RCIM_Cubic, tangentMode=RCTM_Break, time=1, value=0}]}}})
    )
    ApplyForce=false
    ApplyPosition=true
    ApplyPositionFalloff=false
   )
   &SolveForcesAndVelocity(Clamp Velocity=false, Limit Acceleration=false)
   &VC_UpdateRibbon_v1()
   /SetVariables(
    Particles.R1 Color:LinearColor=&ColorFromCurve(ColorCurve=DataInterface() {propertyValues {
       AlphaCurve {keys [<at 0 1 RCIM_Cubic>,<at 0.757 1>,<at 0.8391 0 RCIM_Cubic>]}
       BlueCurve {keys [<at 0.0495 0.0077>,<at 0.4986 1 RCIM_Cubic>]}
       GreenCurve {keys [<at 0.0495 0.0412>,<at 0.4986 0.6526 RCIM_Cubic>]}
       RedCurve {keys [<at 0.0495 1>,<at 0.4986 0.142 RCIM_Cubic>]}
      }})
    Particles.curve_color_lines:LinearColor=&MultiplyLinearColor(
     Linear Color=$Particles.color_lines
     Scale Factor=&MakeVector4_FromLinearColor(LinearColor=&MakeLinearColorFromVectorAndFloat(
       "Vector (RGB)"=<1 1 1>
       "Float (Alpha)"=&FloatFromCurve(FloatCurve=DataInterface() {propertyValues {Curve {keys [{interpMode=RCIM_Cubic, tangentMode=RCTM_Break, time=0.5, value=1}, {interpMode=RCIM_Cubic, tangentMode=RCTM_Break, time=1, value=1}]}}})
      ))
    )
   ) {enabled=false, isSetParameters=true}
   &Color(Color=&MakeCustomLinearColorFromBool(
     Bool=$User.bIsPainted
     True Color=$Particles.Color
     False Color=&ColorFromCurve(ColorCurve=DataInterface() {propertyValues {
        bExposeCurve=true
        AlphaCurve {keys [<at 0.0095 0>,<at 0.0365 1>,<at 1 1 RCIM_Cubic>]}
        BlueCurve {defaultValue=0, keys [<at 0.8537 1>]}
        GreenCurve {defaultValue=0, keys [<at 0.8537 0.7702>]}
        RedCurve {defaultValue=1, keys [<at 0.8537 0.5989>]}
       }})
    )) {enabled=false}
   &ScaleRibbonWidth(Ribbon Width Scale=&FloatFromCurve(
     FloatCurve=DataInterface() {propertyValues {Curve {keys [
         {arriveTangent=3.9698, interpMode=RCIM_Cubic, leaveTangent=3.9698, tangentMode=RCTM_User, time=0.0056, value=0.3729}
         {arriveTangent=1.0516, interpMode=RCIM_Cubic, leaveTangent=1.0516, time=0.2253, value=0.8814}
         {arriveTangent=-0.7998, interpMode=RCIM_Cubic, leaveTangent=-0.7998, time=0.5445, value=0.9396}
         {arriveTangent=-0.9486, arriveTangentWeight=1.0346, tangentMode=RCTM_User, time=0.8327, value=0.3956}
         <at 1.002 0.3176>
        ]}}}
     Scale Curve=1.2
    ))
   /SetVariables(
    Particles.Ribbon_BlackColor:LinearColor=<rgba 0 0 0 1>
    Particles.WaveScaleCurve:NiagaraFloat=&FloatFromCurve(
     FloatCurve=DataInterface() {propertyValues {Curve {keys [
         {arriveTangent=39.5227, interpMode=RCIM_Cubic, leaveTangent=39.5227, tangentMode=RCTM_User, time=0.0092, value=0}
         {
          arriveTangent=0.7219
          arriveTangentWeight=0.0795
          interpMode=RCIM_Cubic
          leaveTangent=0.1806
          leaveTangentWeight=0.0696
          tangentMode=RCTM_Break
          tangentWeightMode=RCTWM_WeightedBoth
          time=0.125
          value=1.25
         }
         {arriveTangent=0.1988, interpMode=RCIM_Cubic, leaveTangent=0.1988, tangentMode=RCTM_User, time=0.45, value=0.625}
        ]}}}
     Scale Curve=10
    )
   ) {isSetParameters=true}
   /SetVariables(
    Particles.CustomPosition_R1:NiagaraPosition=&AddVectorToPosition(
     Position=$Particles.Position
     Vector=&Multiply_VectorByFloat(
      Vector=&MakeVector(
       X=&Sine(
        Normalized Angle=&Add_Float(A=0, B=&Subtract_Float(A=$Emitter.NormalizedLoopAge, B=$Particles.NormalizedAge))
        Period=$User.WavePeriod
        Scale=&Multiply_Float(
         A=&SelectFloatFromArray(Direct Array Index=$Particles.SpawnGroup, Float Selection Array=DataInterface() {propertyValues {FloatData [1,-1]}})
         B=$User.WaveMaxScale
        )
        Bias=0
       )
       Y=&Sine(
        Normalized Angle=&Add_Float(A=0, B=&Subtract_Float(A=$Emitter.NormalizedLoopAge, B=$Particles.NormalizedAge))
        Period=$User.WavePeriod
        Scale=&Multiply_Float(
         A=&SelectFloatFromArray(Direct Array Index=$Particles.SpawnGroup, Float Selection Array=DataInterface() {propertyValues {FloatData [1,-1]}})
         B=$User.WaveMaxScale
        )
        Bias=0
       )
       Z=&Cosine(
        Normalized Angle=&Add_Float(A=0, B=&Subtract_Float(A=$Emitter.NormalizedLoopAge, B=$Particles.NormalizedAge))
        Period=$User.WavePeriod
        Scale=$User.WaveMaxScale
        Bias=0
       )
      )
      Float=$Particles.WaveScaleCurve
     )
    )
    Particles.CustomPosition_R2:NiagaraPosition=&AddVectorToPosition(
     Position=$Particles.Position
     Vector=&Multiply_VectorByFloat(
      Vector=&MakeVector(
       X=&Sine(
        Normalized Angle=&Add_Float(
         A=&Multiply_Float(A=$User.WavePeriod, B=0.5)
         B=&Subtract_Float(A=$Emitter.NormalizedLoopAge, B=$Particles.NormalizedAge)
        )
        Period=$User.WavePeriod
        Scale=&Multiply_Float(
         A=&SelectFloatFromArray(Direct Array Index=$Particles.SpawnGroup, Float Selection Array=DataInterface() {propertyValues {FloatData [1,-1]}})
         B=$User.WaveMaxScale
        )
        Bias=0
       )
       Y=&Sine(
        Normalized Angle=&Add_Float(
         A=&Multiply_Float(A=$User.WavePeriod, B=0.5)
         B=&Subtract_Float(A=$Emitter.NormalizedLoopAge, B=$Particles.NormalizedAge)
        )
        Period=$User.WavePeriod
        Scale=&Multiply_Float(
         A=&SelectFloatFromArray(Direct Array Index=$Particles.SpawnGroup, Float Selection Array=DataInterface() {propertyValues {FloatData [1,-1]}})
         B=$User.WaveMaxScale
        )
        Bias=0
       )
       Z=&Cosine(
        Normalized Angle=&Add_Float(
         A=&Multiply_Float(A=$User.WavePeriod, B=0.5)
         B=&Subtract_Float(A=$Emitter.NormalizedLoopAge, B=$Particles.NormalizedAge)
        )
        Period=$User.WavePeriod
        Scale=$User.WaveMaxScale
        Bias=0
       )
      )
      Float=$Particles.WaveScaleCurve
     )
    )
   ) {isSetParameters=true}
   /SetVariables(Particles.DynamicMP_R2:Vector4f=&MakeVector4(
     X=&FloatFromCurve(FloatCurve=DataInterface() {propertyValues {Curve {
         defaultValue=1
         keys [
          {tangentMode=RCTM_User, time=0, value=0.5}
          {arriveTangent=-0.0633, interpMode=RCIM_Cubic, leaveTangent=-0.0633, tangentMode=RCTM_User, time=0.65, value=0.5}
          {arriveTangent=-0.0699, interpMode=RCIM_Cubic, leaveTangent=-0.0699, tangentMode=RCTM_User, time=0.75, value=-2}
         ]
        }}})
     Y=0
     Z=0
     W=0
    )) {isSetParameters=true}
   &DynamicMaterialParameters(
    Index 0 Param 1=&FloatFromCurve(FloatCurve=DataInterface() {propertyValues {Curve {
        defaultValue=1
        keys [
         {
          arriveTangent=0.0405
          arriveTangentWeight=0.1835
          interpMode=RCIM_Cubic
          leaveTangent=0.0405
          leaveTangentWeight=0.1835
          tangentMode=RCTM_User
          time=0
          value=0
         }
         {
          arriveTangent=0.0201
          arriveTangentWeight=0.1834
          interpMode=RCIM_Cubic
          leaveTangent=0.0201
          leaveTangentWeight=0.1834
          tangentMode=RCTM_User
          time=0.5
          value=1
         }
        ]
       }}})
    Module.Index0Float0Write=true
    Module.Index0Float1Write=true
    Module.Index0Float2Write=true
    Module.Index0Float3Write=true
   )
  ]
  rendererClasses [&NiagaraRibbonRendererProperties]
  renderers [
   {rendererClass=&NiagaraRibbonRendererProperties}
   {rendererClass=&NiagaraRibbonRendererProperties, rendererIndex=1}
   {rendererClass=&NiagaraRibbonRendererProperties, rendererIndex=2}
  ]
 }
}
summary {
 systemName=NS_Boost_HaloFlux_TEST
 emitters [
  {emitterName=Boost_Rings, rendererClasses [&NiagaraMeshRendererProperties, &NiagaraSpriteRendererProperties]}
  {
   bEnabled=false
   emitterName=Boost_Root001
   rendererClasses [&NiagaraMeshRendererProperties, &NiagaraSpriteRendererProperties]
  }
  {emitterName=DiskCenter, rendererClasses [&NiagaraMeshRendererProperties, &NiagaraSpriteRendererProperties]}
  {bEnabled=false, emitterName=NE_BlackFlame_Base1, rendererClasses [&NiagaraRibbonRendererProperties]}
  {emitterName=Ribbon_HaloSpin, rendererClasses [&NiagaraRibbonRendererProperties]}
  {emitterName=NE_BeeHive001, rendererClasses [&NiagaraSpriteRendererProperties, &NiagaraMeshRendererProperties]}
  {
   emitterName=NE_VC_Boost_Sprite_Parent
   rendererClasses [&NiagaraMeshRendererProperties, &NiagaraSpriteRendererProperties]
  }
  {emitterName=NE_VC_Boost_Ribbon_Parent001, rendererClasses [&NiagaraRibbonRendererProperties]}
 ]
 userVariables [
  User.bIsFrontendPreview {type=NiagaraBool}
  User.bIsPainted {type=NiagaraBool}
  User.ColorVariantID {type=NiagaraInt32}
  User.DTCarOffset {type=NiagaraPosition, description="", defaultValue {}}
  User.DTCarOffset003 {type=NiagaraPosition, description="", defaultValue {}}
  User.GameModeID {type=NiagaraInt32}
  User.HighlightPowerTrail {type=NiagaraFloat, defaultValue=2.5, description=""}
  User.MultiVariantColorCone {type=NiagaraFloat, defaultValue=2.5, description=""}
  User.MultiVariantColorTrail {type=NiagaraFloat, defaultValue=1.5, description=""}
  User.PaintedColor {type=LinearColor, description="", defaultValue {struct=LinearColor, value=<rgba 0.8073 0.2503 0 1>}}
  User.PaintedColor_2 {type=LinearColor, description="", defaultValue {struct=LinearColor, value=<rgba 1 0.3512 0 1>}}
  User.PaintedScalar {type=NiagaraFloat, defaultValue=1, description=""}
  User.PortCount {type=NiagaraInt32, defaultValue=1}
  User.Shine {type=NiagaraFloat, defaultValue=15, description=""}
  User.SingleBoostPos {type=NiagaraPosition, description="", defaultValue {}}
  User.SingleBoostPos {type=Vector3f, description="", defaultValue {struct=Vector3f, value=<0 0 0>}}
  User.SingleBoostPos001 {type=NiagaraPosition, description="", defaultValue {}}
  User.SpecMult {type=NiagaraFloat, defaultValue=1.75, description=""}
  User.SpecPower {type=NiagaraFloat, defaultValue=8, description=""}
  User.SPECULAR_FORCE {type=NiagaraFloat, defaultValue=0, description=""}
  User.WaveMaxScale {type=NiagaraFloat, defaultValue=4, description=""}
  User.WavePeriod {type=NiagaraFloat, defaultValue=1, description=""}
  User.WhiteColorTrail {type=NiagaraFloat, defaultValue=1, description=""}
 ]
}
